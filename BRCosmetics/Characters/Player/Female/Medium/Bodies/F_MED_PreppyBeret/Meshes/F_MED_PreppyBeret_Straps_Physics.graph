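# Physics asset for the straps on the F_MED_PreppyBeret female player cosmetic body, built entirely from capsule bodies.
# Each side has a three-segment dynamic chain hanging from a base strap bone, which links to spine_05 alongside spine_04 and upper-arm capsules.
# Body and base links swing and twist freely, while the first two strap joints are tight with zero twist and the tip joint is looser.
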PhysicsAsset /BRCosmetics/Characters/Player/Female/Medium/Bodies/F_MED_PreppyBeret/Meshes/F_MED_PreppyBeret_Straps_Physics
bodies {
 base_strap_l [{shapeType=Capsule, length=0.505, radius=0.505, shapeName=base_strap_l_capsule, rotation=<rot 0 -90 0>}]
 base_strap_r [{shapeType=Capsule, length=0.505, radius=0.505, shapeName=base_strap_r_capsule, rotation=<rot 0 -90 0>}]
 dyn_strap_l_01 [{shapeType=Capsule, length=1, radius=1, center=<1 0 0>, rotation=<rot 90 0 0>}]
 dyn_strap_l_02 [{shapeType=Capsule, length=1, radius=1, rotation=<rot 90 0 0>}]
 dyn_strap_l_03 [{shapeType=Capsule, length=3, radius=2, center=<3 0 0>, rotation=<rot 90 0 0>}]
 dyn_strap_r_01 [{shapeType=Capsule, length=1, radius=1, center=<1 0 0>, rotation=<rot 90 0 0>}]
 dyn_strap_r_02 [{shapeType=Capsule, length=1, radius=1, rotation=<rot 90 0 0>}]
 dyn_strap_r_03 [{shapeType=Capsule, length=2, radius=1, center=<1 0 0.8094>, rotation=<rot 90 0 0>}]
 spine_04 [{shapeType=Capsule, length=4, radius=9, shapeName=spine_04_capsule, center=<0 2.725 0>}]
 spine_05 [{shapeType=Capsule, length=6, radius=9, shapeName=spine_05_capsule, center=<-1.3811 -0.245 0>}]
 upperarm_l [{shapeType=Capsule, length=12.8035, radius=5.1965, center=<10.2197 -0.494 -0.4407>, rotation=<rot 90 180 -175.0002>}]
 upperarm_r [{shapeType=Capsule, length=12.8035, radius=5.1965, center=<-10.2064 0.6462 0.6079>, rotation=<rot 90 180 4.9998>}]
}
constraints [
 {
  bone1Name=spine_05
  bone2Name=spine_04
  swing1LimitDegrees=45
  swing1Motion=Free
  swing2LimitDegrees=45
  swing2Motion=Free
  twistLimitDegrees=45
  twistMotion=Free
 }
 {
  bone1Name=upperarm_l
  bone2Name=spine_05
  swing1LimitDegrees=45
  swing1Motion=Free
  swing2LimitDegrees=45
  swing2Motion=Free
  twistLimitDegrees=45
  twistMotion=Free
 }
 {
  bone1Name=upperarm_r
  bone2Name=spine_05
  swing1LimitDegrees=45
  swing1Motion=Free
  swing2LimitDegrees=45
  swing2Motion=Free
  twistLimitDegrees=45
  twistMotion=Free
 }
 {
  bone1Name=base_strap_l
  bone2Name=spine_05
  swing1LimitDegrees=45
  swing1Motion=Free
  swing2LimitDegrees=45
  swing2Motion=Free
  twistLimitDegrees=45
  twistMotion=Free
 }
 {bone1Name=dyn_strap_l_01, bone2Name=base_strap_l, swing1LimitDegrees=3, swing2LimitDegrees=8, twistLimitDegrees=0}
 {bone1Name=dyn_strap_l_02, bone2Name=dyn_strap_l_01, swing1LimitDegrees=2, swing2LimitDegrees=3, twistLimitDegrees=0}
 {bone1Name=dyn_strap_l_03, bone2Name=dyn_strap_l_02, swing1LimitDegrees=45, swing2LimitDegrees=45, twistLimitDegrees=15}
 {
  bone1Name=base_strap_r
  bone2Name=spine_05
  swing1LimitDegrees=45
  swing1Motion=Free
  swing2LimitDegrees=45
  swing2Motion=Free
  twistLimitDegrees=45
  twistMotion=Free
 }
 {bone1Name=dyn_strap_r_01, bone2Name=base_strap_r, swing1LimitDegrees=2, swing2LimitDegrees=3, twistLimitDegrees=0}
 {bone1Name=dyn_strap_r_02, bone2Name=dyn_strap_r_01, swing1LimitDegrees=2, swing2LimitDegrees=3, twistLimitDegrees=0}
 {bone1Name=dyn_strap_r_03, bone2Name=dyn_strap_r_02, swing1LimitDegrees=45, swing2LimitDegrees=45, twistLimitDegrees=15}
]
