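# Control Rig for the PreppyBeret female character's FaceAcc part that procedurally poses the eight-bone ponytail against body motion.
# Weighted parent constraints move control nulls with the spine, neck, head, clavicles and pelvis, with head-rotation corrections, then TwistBones and B-spline IK shape the chain.
# Spine bend and the IsRigidBodyDisable flag add lower-bone offsets.
# A debug-only function copies body controls to bones; the direct-bone Constrains function is unconnected.
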ControlRigBlueprint /BRCosmetics/Characters/Player/Female/Medium/Bodies/F_MED_PreppyBeret/Meshes/Parts/F_MED_PreppyBeret_FaceAcc_CtrlRig
graphs [
 New Function_ContainedGraph {nodes [
   RigVMFunctionEntryNode()
   > RigVMFunction_Sequence_1 = RigVMAggregateNode()
   &RigUnit_ParentConstraint(
    AdvancedSettings=(RotationOrderForFilter=XZY)
    Child=(Type=Null,Name="dyn_PonyTail_01_ctrl_null")
    Parents=((Item=(Type=Bone,Name="spine_05"),Weight=0),(Item=(Type=Bone,Name="head"),Weight=1))
    Weight=1
    ExecutePin=@RigVMFunction_Sequence_1.A
   )
   > &RigUnit_ParentConstraint(
    AdvancedSettings=(RotationOrderForFilter=XZY)
    Child=(Type=Null,Name="dyn_PonyTail_02_ctrl_null")
    Parents=((Item=(Type=Bone,Name="neck_01"),Weight=0.1101),(Item=(Type=Bone,Name="spine_05"),Weight=0.158),(Item=(Type=Bone,Name="head"),Weight=0.2305),(Item=(Type=Bone,Name="neck_02"),Weight=0.5014))
    Weight=1
   )
   > &RigUnit_ParentConstraint(
    AdvancedSettings=(RotationOrderForFilter=XZY)
    Child=(Type=Null,Name="dyn_PonyTail_03_ctrl_null")
    Parents="((Item=(Type=Bone,Name=\"spine_04\"),Weight=0.0002),(Item=(Type=Bone,Name=\"neck_02\"),Weight=0.0069),(Item=(Type=Bone,Name=\"clavicle_l\"),Weight=0.0177),(Item=(Type=Bone,Name=\"clavicle_r\"),Weight=0.0177),(Item=(Type=Bone,Name=\"neck_01\"),Weight=0.0478),(Item=(T ...(truncated)"
    Weight=1
   )
   > &RigUnit_ParentConstraint(
    AdvancedSettings=(RotationOrderForFilter=XZY)
    Child=(Type=Null,Name="dyn_PonyTail_04_ctrl_null")
    Parents=((Item=(Type=Bone,Name="spine_04"),Weight=0.0086),(Item=(Type=Bone,Name="clavicle_r"),Weight=0.022),(Item=(Type=Bone,Name="clavicle_l"),Weight=0.2394),(Item=(Type=Bone,Name="spine_05"),Weight=0.7301))
    Weight=1
   )
   > &RigUnit_ParentConstraint(
    AdvancedSettings=(RotationOrderForFilter=XZY)
    Child=(Type=Null,Name="dyn_PonyTail_05_ctrl_null")
    Parents=((Item=(Type=Bone,Name="spine_03"),Weight=0.0002),(Item=(Type=Bone,Name="spine_04"),Weight=0.0378),(Item=(Type=Bone,Name="clavicle_l"),Weight=0.2885),(Item=(Type=Bone,Name="spine_05"),Weight=0.6736))
    Weight=1
   )
   > &RigUnit_ParentConstraint(
    AdvancedSettings=(RotationOrderForFilter=XZY)
    Child=(Type=Null,Name="dyn_PonyTail_06_ctrl_null")
    Parents="((Item=(Type=Bone,Name=\"spine_02\"),Weight=0.0005),(Item=(Type=Bone,Name=\"spine_03\"),Weight=0.0248),(Item=(Type=Bone,Name=\"clavicle_l\"),Weight=0.0553),(Item=(Type=Bone,Name=\"clavicle_r\"),Weight=0.0695),(Item=(Type=Bone,Name=\"spine_05\"),Weight=0.4135),(Item= ...(truncated)"
    Weight=1
   )
   > &RigUnit_ParentConstraint(
    AdvancedSettings=(RotationOrderForFilter=XZY)
    Child=(Type=Null,Name="dyn_PonyTail_07_ctrl_null")
    Parents=((Item=(Type=Bone,Name="spine_04"),Weight=0.0013),(Item=(Type=Bone,Name="spine_03"),Weight=0.048),(Item=(Type=Bone,Name="pelvis"),Weight=0.0813),(Item=(Type=Bone,Name="spine_02"),Weight=0.3786),(Item=(Type=Bone,Name="spine_01"),Weight=0.4908))
    Weight=1
   )
   > &RigUnit_ParentConstraint(
    AdvancedSettings=(RotationOrderForFilter=XZY)
    Child=(Type=Null,Name="dyn_PonyTail_08_ctrl_null")
    Parents=((Item=(Type=Bone,Name="spine_03"),Weight=0.0001),(Item=(Type=Bone,Name="spine_02"),Weight=0.0075),(Item=(Type=Bone,Name="spine_01"),Weight=0.1164),(Item=(Type=Bone,Name="pelvis"),Weight=0.876))
    Weight=1
   )
   DeltaRotation v4 = RigVMFunctionReferenceNode(
    Rotation Order=ZYX
    child=(Type=Bone,Name="head")
    outputMinMax=(X=0,Y=1)
    parent=(Type=Null,Name="head_null")
    premult=1
    rotationMinMax=(X=0,Y=-40)
    triggerAxis=Z
    twistAxis=(X=1,Y=0,Z=0)
    ExecuteContext=@RigVMFunction_Sequence_1.B
   )
   > &RigUnit_ModifyTransforms(
    ItemToModify="((Item=(Type=Null,Name=\"dyn_PonyTail_02_ctrl_null\"),Transform=(Rotation=(X=0,Y=0,Z=0,W=1),Translation=(X=2.14,Y=0,Z=4),Scale3D=(X=1,Y=1,Z=1))),(Item=(Type=Null,Name=\"dyn_PonyTail_01_ctrl_null\"),Transform=(Rotation=(X=0,Y=-0.1392,Z=0,W=0.9903),Translation=( ...(truncated)"
    WeightMinimum=0
    Weight=@"DeltaRotation v4".outputWeight
   )
   RigVMFunctionReturnNode(ExecuteContext=@RigVMFunction_Sequence_1.E)
   n1 = &RigUnit_ItemArray(Items="((Type=Bone,Name=\"dyn_PonyTail_01\"),(Type=Bone,Name=\"dyn_PonyTail_02\"),(Type=Bone,Name=\"dyn_PonyTail_03\"),(Type=Bone,Name=\"dyn_PonyTail_04\"),(Type=Bone,Name=\"dyn_PonyTail_05\"),(Type=Bone,Name=\"dyn_PonyTail_06\"),(Type=Bone,Name=\"dyn_PonyTail_07\"),(Type=Bone ...(truncated)")
   TwistBones = RigVMFunctionReferenceNode(
    Axis to Twist=(bX=true,bY=true,bZ=true)
    End=(Type=Bone,Name="neck_02")
    Start=(Type=Bone,Name="spine_02")
    Bones to Twist=@n1.Items
    ExecuteContext=@RigVMFunction_Sequence_1.C
   )
   RigUnit_ItemArray = &RigUnit_ItemArray(Items="((Type=Control,Name=\"dyn_PonyTail_01_ctrl\"),(Type=Control,Name=\"dyn_PonyTail_02_ctrl\"),(Type=Control,Name=\"dyn_PonyTail_03_ctrl\"),(Type=Control,Name=\"dyn_PonyTail_04_ctrl\"),(Type=Control,Name=\"dyn_PonyTail_06_ctrl\"),(Type=Control,Name=\"dyn_PonyTail_05_ctrl ...(truncated)")
   > SplineIK = RigVMFunctionReferenceNode(
    Primary Axis=(X=1,Y=0,Z=0)
    Samples Per Segment=1
    Secondary Spline Direction=(X=0,Y=0,Z=0)
    Spline Mode=BSpline
    Up Axis=(X=0,Y=0,Z=0)
    debug=false
    Controls=@RigUnit_ItemArray.Items
    Bones=@n1.Items
   )
   DeltaRotation v4_1 = RigVMFunctionReferenceNode(
    Rotation Order=ZYX
    child=(Type=Bone,Name="spine_03")
    debugRotation=false
    onlyTwist=false
    outputMinMax=(X=0,Y=1)
    parent=(Type=Bone,Name="spine_01")
    premult=1
    rotationMinMax=(X=-10,Y=-40)
    triggerAxis=Z
    twistAxis=(X=1,Y=0,Z=0)
    ExecuteContext=@RigVMFunction_Sequence_1.D
   )
   > &RigUnit_OffsetTransformForItem(
    Item=(Type=Bone,Name="dyn_PonyTail_07")
    OffsetTransform=(Rotation=(X=0,Y=-0.342,Z=0,W=0.9397),Translation=(X=-1.8,Y=0,Z=2.2),Scale3D=(X=1,Y=1,Z=1))
    bPropagateToChildren=false
    Weight=@"DeltaRotation v4_1".outputWeight
   )
   RigVMFunction_MathBoolToFloat = &RigVMFunction_MathBoolToFloat(Value=RigVMVariableNode() {variable=IsRigidBodyDisable}.Value)
   > &RigUnit_OffsetTransformForItem(
    Item=(Type=Bone,Name="dyn_PonyTail_06")
    OffsetTransform=(Rotation=(X=0,Y=0,Z=0,W=1),Translation=(X=0,Y=0,Z=-4),Scale3D=(X=1,Y=1,Z=1))
    bPropagateToChildren=false
    Weight=@RigVMFunction_MathBoolToFloat.Result
   )
   > &RigUnit_OffsetTransformForItem(
    Item=(Type=Bone,Name="dyn_PonyTail_05")
    OffsetTransform=(Rotation=(X=0,Y=0,Z=0,W=1),Translation=(X=0,Y=0,Z=-2),Scale3D=(X=1,Y=1,Z=1))
    bPropagateToChildren=false
    Weight=&RigVMFunction_MathBoolToFloat(Value=RigVMVariableNode() {variable=IsRigidBodyDisable}.Value).Result
   )
  ]}
 New Function_ContainedGraph {nodes [
   RigVMFunctionEntryNode()
   > &RigUnit_ParentConstraint(
    AdvancedSettings=(RotationOrderForFilter=XZY)
    Child=(Type=Bone,Name="dyn_PonyTail_01")
    Parents=((Item=(Type=Bone,Name="neck_01"),Weight=0.091),(Item=(Type=Bone,Name="spine_05"),Weight=0.114),(Item=(Type=Bone,Name="head"),Weight=0.3809),(Item=(Type=Bone,Name="neck_02"),Weight=0.414))
    Weight=1
   )
   > &RigUnit_ParentConstraint(
    AdvancedSettings=(RotationOrderForFilter=XZY)
    Child=(Type=Bone,Name="dyn_PonyTail_02")
    Parents=((Item=(Type=Bone,Name="neck_01"),Weight=0.1431),(Item=(Type=Bone,Name="spine_05"),Weight=0.2053),(Item=(Type=Bone,Name="head"),Weight=0.2305),(Item=(Type=Bone,Name="neck_02"),Weight=0.421))
    Weight=1
   )
   > &RigUnit_ParentConstraint(
    AdvancedSettings=(RotationOrderForFilter=XZY)
    Child=(Type=Bone,Name="dyn_PonyTail_03")
    Parents="((Item=(Type=Bone,Name=\"neck_02\"),Weight=0.0069),(Item=(Type=Bone,Name=\"clavicle_l\"),Weight=0.0374),(Item=(Type=Bone,Name=\"clavicle_r\"),Weight=0.0451),(Item=(Type=Bone,Name=\"neck_01\"),Weight=0.0478),(Item=(Type=Bone,Name=\"spine_03\"),Weight=0.0598),(Item=(T ...(truncated)"
    Weight=1
   )
   > &RigUnit_ParentConstraint(
    AdvancedSettings=(RotationOrderForFilter=XZY)
    Child=(Type=Bone,Name="dyn_PonyTail_04")
    Parents="((Item=(Type=Bone,Name=\"upperarm_twist_01_r\"),Weight=0.0001),(Item=(Type=Bone,Name=\"spine_04\"),Weight=0.0278),(Item=(Type=Bone,Name=\"clavicle_r\"),Weight=0.188),(Item=(Type=Bone,Name=\"clavicle_l\"),Weight=0.1888),(Item=(Type=Bone,Name=\"spine_05\"),Weight=0.59 ...(truncated)"
    Weight=1
   )
   > &RigUnit_ParentConstraint(
    AdvancedSettings=(RotationOrderForFilter=XZY)
    Child=(Type=Bone,Name="dyn_PonyTail_05")
    Parents="((Item=(Type=Bone,Name=\"upperarm_twist_01_r\"),Weight=0.0001),(Item=(Type=Bone,Name=\"spine_04\"),Weight=0.0278),(Item=(Type=Bone,Name=\"clavicle_r\"),Weight=0.188),(Item=(Type=Bone,Name=\"clavicle_l\"),Weight=0.1888),(Item=(Type=Bone,Name=\"spine_05\"),Weight=0.59 ...(truncated)"
    Weight=1
   )
   > RigVMFunctionReturnNode()
  ]}
 New Function_ContainedGraph {nodes [
   n1 = RigVMFunctionEntryNode()
   RigVMFunction_ControlFlowBranch = &RigVMFunction_ControlFlowBranch(ExecuteContext=@n1.ExecuteContext, Condition=@n1.debug)
   RigVMFunctionReturnNode(ExecuteContext=@RigVMFunction_ControlFlowBranch.Completed)
   RigUnit_ItemArray = &RigUnit_ItemArray(Items="((Type=Bone,Name=\"spine_01\"),(Type=Bone,Name=\"spine_02\"),(Type=Bone,Name=\"spine_03\"),(Type=Bone,Name=\"spine_04\"),(Type=Bone,Name=\"spine_05\"),(Type=Bone,Name=\"neck_01\"),(Type=Bone,Name=\"neck_02\"),(Type=Bone,Name=\"head\"),(Type=Bone,Name=\"clavicle_r\"),(Type=B ...(truncated)")
   DISPATCH_RigVMDispatch_ArrayIterator = RigVMDispatchNode(Array=@RigUnit_ItemArray.Items, ExecuteContext=@RigVMFunction_ControlFlowBranch.True) {template="DISPATCH_RigVMDispatch_ArrayIterator(in Array,out Element,out Index,out Count,out Ratio)"}
   Concat = /Script/RigVM.RigVMFunction_NameConcat(B=_ctrl, A=@DISPATCH_RigVMDispatch_ArrayIterator.Element.Name) {is=struct RigVMUnitNode, template="Concat::Execute(in A,in B,out Result)"}
   > &RigUnit_SetTransform(
    Item=(Type=Bone,Name="spine_01")
    Weight=1
    bInitial=False
    Item.Name=@DISPATCH_RigVMDispatch_ArrayIterator.Element.Name
    Value=&RigUnit_GetTransform(Item=(Type=Control,Name="spine_01_ctrl"), bInitial=False, Item.Name=@Concat.Result).Transform
   )
  ]}
 New Function_ContainedGraph {nodes [
   n1 = RigVMFunctionEntryNode()
   RigVMFunction_MathBoolToInteger = &RigVMFunction_MathBoolToInteger(Value=@n1.onlyTwist)
   RigVMFunction_MathQuaternionSwingTwist = &RigVMFunction_MathQuaternionSwingTwist(
    TwistAxis=@n1.twistAxis
    Input=&RigUnit_GetRelativeTransformForItem(bChildInitial=False, bParentInitial=False, Child=@n1.child, Parent=@n1.parent).RelativeTransform.Rotation
   )
   n2 = RigVMDispatchNode(
    Values=((),())
    Index=@RigVMFunction_MathBoolToInteger.Result
    Values.0=@RigVMFunction_MathQuaternionSwingTwist.Swing
    Values.1=@RigVMFunction_MathQuaternionSwingTwist.Twist
   ) {template="DISPATCH_RigVMDispatch_SelectInt32(in Index,in Values,out Result)"}
   RigVMFunction_MathQuaternionToEuler = &RigVMFunction_MathQuaternionToEuler(Value=@n2.Result, RotationOrder=@n1.Rotation Order)
   > DISPATCH_RigVMDispatch_Print = RigVMDispatchNode(ScreenColor=(R=1,G=1,B=1,A=1), ScreenDuration=1.5, Enabled=@n1.debugRotation, Value=@RigVMFunction_MathQuaternionToEuler.Result) {template="DISPATCH_RigVMDispatch_Print(in Prefix,in Value,in Enabled,in ScreenDuration,in ScreenColor)"}
   Cast = RigVMDispatchNode(Value=@n1.triggerAxis) {template="DISPATCH_RigVMDispatch_CastEnumToInt(in Value,out Result)"}
   DISPATCH_RigVMDispatch_SelectInt32 = RigVMDispatchNode(
    Values=(0,0,0)
    Index=@Cast.Result
    Values.0=@RigVMFunction_MathQuaternionToEuler.Result.X
    Values.1=@RigVMFunction_MathQuaternionToEuler.Result.Y
    Values.2=@RigVMFunction_MathQuaternionToEuler.Result.Z
   ) {template="DISPATCH_RigVMDispatch_SelectInt32(in Index,in Values,out Result)"}
   Remap = &RigVMFunction_MathDoubleRemap(
    bClamp=False
    Value=@DISPATCH_RigVMDispatch_SelectInt32.Result
    TargetMinimum=@n1.outputMinMax.X
    TargetMaximum=@n1.outputMinMax.Y
    SourceMinimum=@n1.rotationMinMax.X
    SourceMaximum=@n1.rotationMinMax.Y
   )
   Absolute = &RigVMFunction_MathDoubleAbs(Value=@Remap.Result)
   Multiply = &RigVMFunction_MathDoubleMul(
    A=&RigVMFunction_MathDoubleClamp(Value=RigVMDispatchNode(True=@Absolute.Result, Condition=@n1.absoluteValue, False=@Remap.Result) {template="DISPATCH_RigVMDispatch_If(in Condition,in True,in False,out Result)"}.Result).Result
    B=@n1.premult
   )
   > RigVMFunctionReturnNode(outputQuat=@n2.Result, outputWeight=@Multiply.Result)
  ]}
 RigVMFunctionLibrary {nodes [
   Constrains = RigVMCollapseNode(ExecuteContext=())
   DeltaRotation v4 = RigVMCollapseNode(
    ExecuteContext=()
    Rotation Order=ZYX
    child=(Type=Bone,Name="None")
    outputMinMax=(X=0,Y=1)
    parent=(Type=Bone,Name="None")
    premult=1
    rotationMinMax=(X=0,Y=0)
    twistAxis=(X=1,Y=0,Z=0)
   )
   controllers = RigVMCollapseNode(ExecuteContext=(), debug=False)
   ponytail = RigVMCollapseNode(ExecuteContext=())
  ]}
 RigVMFunction_Sequence_2_ContainedGraph {nodes [
   n1 = RigVMFunctionEntryNode()
   n2 = &RigVMFunction_Sequence(ExecuteContext=@n1.ExecuteContext)
   n3 = &RigVMFunction_Sequence(ExecuteContext=@n2.B)
   n4 = &RigVMFunction_Sequence(ExecuteContext=@n3.B)
   n5 = &RigVMFunction_Sequence(ExecuteContext=@n4.B)
   RigVMFunctionReturnNode(A=@n2.A, B=@n3.A, C=@n4.A, D=@n5.A, E=@n5.B)
  ]}
 RigVMModel {nodes [
   RigUnit_BeginExecution = &RigUnit_BeginExecution()
   controllers = RigVMFunctionReferenceNode(debug=false, ExecuteContext=@RigUnit_BeginExecution.ExecutePin)
   > ponytail = RigVMFunctionReferenceNode()
   Constrains = RigVMFunctionReferenceNode(ExecuteContext=())
  ]}
]
hierarchy [
 root {loc=<0 0 0>, rot=<0 0 0 1>}
 pelvis {parent=root, loc=<0 -0.5301 98.0571>, rot=<0.0127 0.707 -0.0127 -0.707>}
 spine_01 {parent=pelvis, loc=<6.1208 -0.9328 0>, rot=<0 0 0.034 -0.9994>}
 spine_02 {parent=spine_01, loc=<6.5997 -0.2095 0>, rot=<0 0 0.0238 -0.9997>}
 spine_03 {parent=spine_02, loc=<6.4914 -0.1237 0>, rot=<0 0 0.0149 -0.9999>}
 spine_04 {parent=spine_03, loc=<6.9598 0.0794 0>, rot=<0 0 -0.0132 -0.9999>}
 spine_05 {parent=spine_04, loc=<6.9987 0 0>, rot=<0 0 -0.0015 -1>}
 neck_01 {parent=spine_05, loc=<11.3138 0.1529 0>, rot=<0 0 -0.2003 -0.9797>}
 neck_02 {parent=neck_01, loc=<4.5627 0 0>, rot=<0 0 0.0504 -0.9987>}
 head {parent=neck_02, loc=<5.3914 0 0>, rot=<0 0 0.115 -0.9934>}
 base_PonyTail {parent=head, loc=<-0.3471 -6.1639 0>, rot=<0.1631 -0.688 0.688 -0.1631>}
 dyn_PonyTail_01 {parent=base_PonyTail, loc=<0 0 0>, rot=<0 0 0 1>}
 dyn_PonyTail_02 {parent=dyn_PonyTail_01, loc=<4.5403 0 0>, rot=<0 0.0013 0 -1>}
 dyn_PonyTail_03 {parent=dyn_PonyTail_02, loc=<5.6233 0 0>, rot=<0 -0.0749 0 -0.9972>}
 dyn_PonyTail_04 {parent=dyn_PonyTail_03, loc=<5.5562 0 0>, rot=<0 -0.0572 0 -0.9984>}
 dyn_PonyTail_05 {parent=dyn_PonyTail_04, loc=<5.5597 0 0>, rot=<0 0.0107 0 -0.9999>}
 dyn_PonyTail_06 {parent=dyn_PonyTail_05, loc=<6.691 0 0>, rot=<0 -0.0343 0 -0.9994>}
 dyn_PonyTail_07 {parent=dyn_PonyTail_06, loc=<10.2706 0 -1.2803>, rot=<0.0241 -0.1003 -0.0162 -0.9945>}
 dyn_PonyTail_08 {parent=dyn_PonyTail_07, loc=<10.9188 0 0>, rot=<-0.0256 0.004 0.0138 -0.9996>}
 Base_FrontHair {parent=head, loc=<9.446 8.485 0.2656>, rot=<-0.1004 -0.6987 -0.6931 -0.146>}
 dyn_FrontHair_01 {parent=Base_FrontHair, loc=<0 0 0>, rot=<0 0 0 1>}
 dyn_FrontHair_02 {parent=dyn_FrontHair_01, loc=<1.9449 0 0>, rot=<0.0096 -0.0537 0.0396 -0.9977>}
 base_sideHair_b_l {parent=head, loc=<6.3962 3.6571 -7.1484>, rot=<0.1696 -0.8225 0.3027 -0.4505>}
 dyn_sideHair_b_l_01 {parent=base_sideHair_b_l, loc=<0 0 0>, rot=<0 0 0 1>}
 dyn_sideHair_b_l_02 {parent=dyn_sideHair_b_l_01, loc=<3.9742 0 0>, rot=<-0.0032 0.3055 0.012 -0.9521>}
 dyn_sideHair_b_l_03 {parent=dyn_sideHair_b_l_02, loc=<2.1348 0 0>, rot=<0.0238 0.4081 0.0469 -0.9114>}
 base_sideHair_a_l {parent=head, loc=<2.4467 3.9487 -6.4655>, rot=<0.119 0.6046 0.7823 0.0906>}
 dyn_sideHair_a_l_01 {parent=base_sideHair_a_l, loc=<0 0 0>, rot=<0 0 0 1>}
 dyn_sideHair_a_l_02 {parent=dyn_sideHair_a_l_01, loc=<2.0751 0 0>, rot=<0.0689 0.2938 -0.0479 -0.9522>}
 Base_SideHair_l_c {parent=head, loc=<2.78 0.841 -7.6564>, rot=<0.0738 0.0426 0.9964 -0.001>}
 dyn_SideHair_l_c_01 {parent=Base_SideHair_l_c, loc=<0 0 0>, rot=<0 0 0 1>}
 dyn_SideHair_l_c_02 {parent=dyn_SideHair_l_c_01, loc=<4.0525 0 0>, rot=<-0.4135 0.1687 0.1449 -0.8829>}
 base_SideHair_l_d {parent=head, loc=<0.5779 -1.7378 -8.1619>, rot=<0.0263 -0.9143 0.4007 -0.053>}
 dyn_SideHair_l_d_01 {parent=base_SideHair_l_d, loc=<0 0 0>, rot=<0 0 0 1>}
 dyn_SideHair_l_d_02 {parent=dyn_SideHair_l_d_01, loc=<3.1464 0 0>, rot=<0.2541 0.1824 -0.0793 -0.9465>}
 dyn_SideHair_l_d_03 {parent=dyn_SideHair_l_d_02, loc=<5.1434 0 0>, rot=<-0.1725 -0.2971 -0.0274 -0.9387>}
 base_SideHair_r_a {parent=head, loc=<4.7326 3.5336 6.4153>, rot=<0.0658 0.6309 0.7713 0.0527>}
 dyn_SideHair_r_a_01 {parent=base_SideHair_r_a, loc=<0 0 0>, rot=<0 0 0 1>}
 dyn_SideHair_r_a_02 {parent=dyn_SideHair_r_a_01, loc=<3.495 0 0>, rot=<-0.2328 0.2497 0.1091 -0.9336>}
 base_SideHair_l_e {parent=head, loc=<4.2617 -4.3259 -7.2841>, rot=<0.1543 -0.9329 0.3241 -0.0302>}
 dyn_SideHair_l_e_01 {parent=base_SideHair_l_e, loc=<0 0 0>, rot=<0 0 0 1>}
 dyn_SideHair_l_e_02 {parent=dyn_SideHair_l_e_01, loc=<2.5351 0 0>, rot=<-0.0011 -0.12 -0.1323 -0.9839>}
 dyn_SideHair_l_e_03 {parent=dyn_SideHair_l_e_02, loc=<2.2081 0 0>, rot=<0.0372 0.1233 -0.2444 -0.9611>}
 base_SideHair_r_c {parent=head, loc=<1.8713 -3.0919 7.5525>, rot=<0.2275 -0.6582 0.6833 -0.2194>}
 dyn_SideHair_r_c_01 {parent=base_SideHair_r_c, loc=<0 0 0>, rot=<0 0 0 1>}
 dyn_SideHair_r_c_02 {parent=dyn_SideHair_r_c_01, loc=<3.4901 0 0>, rot=<0.1299 0.2131 -0.149 -0.9568>}
 base_SideHair_r_b {parent=head, loc=<3.0921 0.7942 7.8261>, rot=<-0.0284 -0.8522 -0.5201 -0.0502>}
 dyn_SideHair_r_b_01 {parent=base_SideHair_r_b, loc=<0 0 0>, rot=<0 0 0 1>}
 dyn_SideHair_r_b_02 {parent=dyn_SideHair_r_b_01, loc=<5.1282 0 0>, rot=<-0.0953 0.1691 0.0288 -0.9806>}
 base_SideHair_r_d {parent=head, loc=<8.1737 3.4828 6.2616>, rot=<0.4922 -0.2784 0.8062 -0.174>}
 dyn_SideHair_r_d_01 {parent=base_SideHair_r_d, loc=<0 0 0>, rot=<0 0 0 1>}
 dyn_SideHair_r_d_02 {parent=dyn_SideHair_r_d_01, loc=<4.5133 0 0>, rot=<0.0009 0.4444 -0.2211 -0.8681>}
 base_sunGlasses {parent=head, loc=<3.6036 0.6551 0>, rot=<0.0029 -0.7071 -0.0029 -0.7071>}
 dyn_sunGlasses {parent=base_sunGlasses, loc=<0 0 0>, rot=<0 0 0 1>}
 clavicle_l {parent=spine_05, loc=<5.4831 -0.1053 -3.2946>, rot=<0.0465 -0.7289 -0.009 -0.6829>}
 upperarm_l {parent=clavicle_l, loc=<9.8658 -1.2136 -0.4105>, rot=<0.0164 -0.3245 -0.0269 -0.9454>}
 lowerarm_l {parent=upperarm_l, loc=<26.465 -0.8345 0.259>, rot=<0.0036 0.0161 -0.2298 -0.9731>}
 clavicle_r {parent=spine_05, loc=<5.4831 -0.1053 3.2946>, rot=<0.7289 0.0465 -0.6829 0.009>}
 upperarm_r {parent=clavicle_r, loc=<-9.8658 1.2136 0.4105>, rot=<0.0164 -0.3245 -0.0269 -0.9454>}
 lowerarm_r {parent=upperarm_r, loc=<-26.465 0.8345 -0.259>, rot=<0.0036 0.0161 -0.2298 -0.9731>}
 head_null {parent=spine_05, type=Null, loc=<20.6573 3.5489 0>, rot=<0 0 -0.036 -0.9994>}
 spine_01_ctrl {controlType=EULER_TRANSFORM, type=Control, loc=<0 0 0>, rot=<0 0 0 1>}
 spine_02_ctrl {controlType=EULER_TRANSFORM, parent=spine_01_ctrl, type=Control, loc=<0 0 0>, rot=<0 0 0 1>}
 spine_03_ctrl {controlType=EULER_TRANSFORM, parent=spine_02_ctrl, type=Control, loc=<0 0 0>, rot=<0 0 0 1>}
 spine_04_ctrl {controlType=EULER_TRANSFORM, parent=spine_03_ctrl, type=Control, loc=<0 0 0>, rot=<0 0 0 1>}
 spine_05_ctrl {controlType=EULER_TRANSFORM, parent=spine_04_ctrl, type=Control, loc=<0 0 0>, rot=<0 0 0 1>}
 neck_01_ctrl {controlType=EULER_TRANSFORM, parent=spine_05_ctrl, type=Control, loc=<0 0 0>, rot=<0 0 0 1>}
 neck_02_ctrl {controlType=EULER_TRANSFORM, parent=neck_01_ctrl, type=Control, loc=<0 0 0>, rot=<0 0 0 1>}
 head_ctrl {controlType=EULER_TRANSFORM, parent=neck_02_ctrl, type=Control, loc=<0 0 0>, rot=<0 0 0 1>}
 clavicle_r_ctrl {controlType=EULER_TRANSFORM, parent=spine_05_ctrl, type=Control, loc=<0 0 0>, rot=<0 0 0 1>}
 clavicle_l_ctrl {controlType=EULER_TRANSFORM, parent=spine_05_ctrl, type=Control, loc=<0 0 0>, rot=<0 0 0 1>}
 dyn_PonyTail_07_ctrl_null {type=Null, loc=<0 -18.1193 114.839>, rot=<-0.511 -0.511 0.4887 -0.4887>}
 dyn_PonyTail_07_ctrl {
  controlType=EULER_TRANSFORM
  parent=dyn_PonyTail_07_ctrl_null
  type=Control
  loc=<-32.3512 -18.1193 110.6891>
  rot=<0.5558 0.4616 -0.4372 0.5357>
 }
 dyn_PonyTail_06_ctrl_null {type=Null, loc=<0 -16.5648 125.2443>, rot=<-0.4616 -0.4616 0.5357 -0.5357>}
 dyn_PonyTail_06_ctrl {
  controlType=EULER_TRANSFORM
  parent=dyn_PonyTail_06_ctrl_null
  type=Control
  loc=<-15.2737 -16.5648 124.4894>
  rot=<0.4797 0.4429 -0.5195 0.5512>
 }
 dyn_PonyTail_05_ctrl_null {type=Null, loc=<0 -15.1242 131.7784>, rot=<-0.4429 -0.4429 0.5512 -0.5512>}
 dyn_PonyTail_05_ctrl {
  controlType=EULER_TRANSFORM
  parent=dyn_PonyTail_05_ctrl_null
  type=Control
  loc=<-2.7359 -15.1242 131.8673>
  rot=<0.437 0.4488 -0.5559 0.5464>
 }
 dyn_PonyTail_04_ctrl_null {type=Null, loc=<0 -14.0438 137.2321>, rot=<-0.4488 -0.4488 0.5464 -0.5464>}
 dyn_PonyTail_04_ctrl {
  controlType=EULER_TRANSFORM
  parent=dyn_PonyTail_04_ctrl_null
  type=Control
  loc=<-21.1858 -14.0438 135.7007>
  rot=<0.4793 0.4168 -0.5199 0.5712>
 }
 dyn_PonyTail_03_ctrl_null {type=Null, loc=<0 -12.3489 142.5235>, rot=<-0.4168 -0.4168 0.5712 -0.5712>}
 dyn_PonyTail_03_ctrl {
  controlType=EULER_TRANSFORM
  parent=dyn_PonyTail_03_ctrl_null
  type=Control
  loc=<-26.8512 -12.3489 140.0842>
  rot=<0.4584 0.3729 -0.5384 0.6008>
 }
 dyn_PonyTail_02_ctrl_null {type=Null, loc=<0 -9.8528 147.5624>, rot=<0.3729 0.3729 -0.6008 0.6008>}
 dyn_PonyTail_02_ctrl {
  controlType=EULER_TRANSFORM
  parent=dyn_PonyTail_02_ctrl_null
  type=Control
  loc=<-4.1484 -9.8528 147.5739>
  rot=<0.3721 0.3737 -0.6013 0.6003>
 }
 dyn_PonyTail_01_ctrl_null {type=Null, loc=<0 -7.8482 151.6362>, rot=<-0.3737 -0.3737 0.6003 -0.6003>}
 dyn_PonyTail_01_ctrl {controlType=EULER_TRANSFORM, parent=dyn_PonyTail_01_ctrl_null, type=Control, loc=<0 0 0>, rot=<0 0 0 1>}
 dyn_PonyTail_08_ctrl_null {type=Null, loc=<0 -17.5769 102.4177>, rot=<0.511 0.511 -0.4887 0.4887>}
 dyn_PonyTail_08_ctrl {controlType=EULER_TRANSFORM, parent=dyn_PonyTail_08_ctrl_null, type=Control, loc=<0 0 0>, rot=<0 0 0 1>}
]
variables [
 IsRigidBodyDisable {type=bool, default=False}
 is Backbling {type=bool, default=False}
]
