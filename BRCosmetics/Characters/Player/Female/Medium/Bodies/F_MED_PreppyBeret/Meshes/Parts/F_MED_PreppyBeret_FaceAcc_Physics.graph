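# Physics asset for the face accessory mesh of the female medium PreppyBeret player cosmetic, set up for ponytail motion.
# A seven-link dyn_PonyTail capsule chain hangs from base_PonyTail on the neck, with tight swing and twist limits that loosen at the last two links.
# Capsules on the pelvis, spine, neck and arms form a body proxy, linked by constraints with free swing and twist.
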PhysicsAsset /BRCosmetics/Characters/Player/Female/Medium/Bodies/F_MED_PreppyBeret/Meshes/Parts/F_MED_PreppyBeret_FaceAcc_Physics
bodies {
 base_PonyTail [{shapeType=Capsule, length=0.505, radius=0.505, shapeName=base_PonyTail_capsule, rotation=<rot 0 -90 0>}]
 dyn_PonyTail_01 [{shapeType=Capsule, length=2, radius=2, center=<1 0 0>, rotation=<rot 90 0 0>}]
 dyn_PonyTail_02 [{shapeType=Capsule, length=2, radius=2, center=<2 0 0.414>, rotation=<rot 90 0 0>}]
 dyn_PonyTail_03 [{shapeType=Capsule, length=2, radius=2, center=<2 0 0.414>, rotation=<rot 90 0 0>}]
 dyn_PonyTail_04 [{shapeType=Capsule, length=2, radius=2, center=<2 0 0.3803>, rotation=<rot 90 0 0>}]
 dyn_PonyTail_05 [{shapeType=Capsule, length=2, radius=2, center=<2 0 0>, rotation=<rot 90 0 0>}]
 dyn_PonyTail_06 [{shapeType=Capsule, length=2, radius=2, center=<2 0 0>, rotation=<rot 90 0 0>}]
 dyn_PonyTail_07 [{shapeType=Capsule, length=2, radius=2, center=<3 0 0>, rotation=<rot 90 0 0>}]
 lowerarm_l [{shapeType=Capsule, length=24.3421, radius=5.8419, shapeName=lowerarm_l_capsule, center=<18.8407 0 0>, rotation=<rot 90 0 0>}]
 lowerarm_r [{shapeType=Capsule, length=24.3421, radius=5.8419, shapeName=lowerarm_r_capsule, center=<-18.8407 0 0>, rotation=<rot 90 0 180>}]
 neck_01 [{shapeType=Capsule, length=6, radius=4.944, shapeName=neck_01_capsule, center=<3.2477 0 0>, rotation=<rot 90 167.6609 159.1678>}]
 pelvis [{shapeType=Capsule, length=6.8753, radius=12.4124, shapeName=pelvis_capsule, center=<-1.1109 -0.7166 0>, rotation=<rot 0 -90 0>}]
 spine_01 [{shapeType=Capsule, length=9, radius=10, shapeName=spine_01_capsule, center=<1 -0.3955 0>, rotation=<rot 0 -90 0>}]
 spine_02 [{shapeType=Capsule, length=9, radius=10, shapeName=spine_02_capsule, center=<1 -0.3933 0>, rotation=<rot 0 -90 0>}]
 spine_03 [{shapeType=Capsule, length=9, radius=10, shapeName=spine_03_capsule, center=<1 -0.9618 0>, rotation=<rot 0 -90 0>}]
 spine_04 [{shapeType=Capsule, length=9, radius=10, shapeName=spine_04_capsule, center=<8.3329 0.5205 0>, rotation=<rot 0 -90 0>}]
 spine_05 [{shapeType=Capsule, length=14.572, radius=10, shapeName=spine_05_capsule, center=<5.4319 2.6923 0>, rotation=<rot 0 -90 0>}]
 upperarm_l [{shapeType=Capsule, length=20, radius=6, shapeName=upperarm_l_capsule, center=<15 0 0>, rotation=<rot 90 180 0>}]
 upperarm_r [{shapeType=Capsule, length=20, radius=6, shapeName=upperarm_r_capsule, center=<-15 0 0>, rotation=<rot 90 0 0>}]
}
constraints [
 {
  bone1Name=spine_01
  bone2Name=pelvis
  swing1LimitDegrees=45
  swing1Motion=Free
  swing2LimitDegrees=45
  swing2Motion=Free
  twistLimitDegrees=45
  twistMotion=Free
 }
 {
  bone1Name=spine_02
  bone2Name=spine_01
  swing1LimitDegrees=45
  swing1Motion=Free
  swing2LimitDegrees=45
  swing2Motion=Free
  twistLimitDegrees=45
  twistMotion=Free
 }
 {
  bone1Name=spine_03
  bone2Name=spine_02
  swing1LimitDegrees=45
  swing1Motion=Free
  swing2LimitDegrees=45
  swing2Motion=Free
  twistLimitDegrees=45
  twistMotion=Free
 }
 {
  bone1Name=spine_04
  bone2Name=spine_03
  swing1LimitDegrees=45
  swing1Motion=Free
  swing2LimitDegrees=45
  swing2Motion=Free
  twistLimitDegrees=45
  twistMotion=Free
 }
 {
  bone1Name=spine_05
  bone2Name=spine_04
  swing1LimitDegrees=45
  swing1Motion=Free
  swing2LimitDegrees=45
  swing2Motion=Free
  twistLimitDegrees=45
  twistMotion=Free
 }
 {
  bone1Name=upperarm_r
  bone2Name=spine_05
  swing1LimitDegrees=45
  swing1Motion=Free
  swing2LimitDegrees=45
  swing2Motion=Free
  twistLimitDegrees=45
  twistMotion=Free
 }
 {
  bone1Name=upperarm_l
  bone2Name=spine_05
  swing1LimitDegrees=45
  swing1Motion=Free
  swing2LimitDegrees=45
  swing2Motion=Free
  twistLimitDegrees=45
  twistMotion=Free
 }
 {
  bone1Name=base_PonyTail
  bone2Name=neck_01
  swing1LimitDegrees=45
  swing1Motion=Free
  swing2LimitDegrees=45
  swing2Motion=Free
  twistLimitDegrees=45
  twistMotion=Free
 }
 {bone1Name=dyn_PonyTail_01, bone2Name=base_PonyTail, swing1LimitDegrees=15, swing2LimitDegrees=8, twistLimitDegrees=3}
 {bone1Name=dyn_PonyTail_02, bone2Name=dyn_PonyTail_01, swing1LimitDegrees=12, swing2LimitDegrees=12, twistLimitDegrees=5}
 {bone1Name=dyn_PonyTail_03, bone2Name=dyn_PonyTail_02, swing1LimitDegrees=12, swing2LimitDegrees=12, twistLimitDegrees=5}
 {bone1Name=dyn_PonyTail_04, bone2Name=dyn_PonyTail_03, swing1LimitDegrees=12, swing2LimitDegrees=12, twistLimitDegrees=5}
 {bone1Name=dyn_PonyTail_05, bone2Name=dyn_PonyTail_04, swing1LimitDegrees=12, swing2LimitDegrees=12, twistLimitDegrees=5}
 {bone1Name=dyn_PonyTail_06, bone2Name=dyn_PonyTail_05, swing1LimitDegrees=25, swing2LimitDegrees=30, twistLimitDegrees=5}
 {bone1Name=dyn_PonyTail_07, bone2Name=dyn_PonyTail_06, swing1LimitDegrees=25, swing2LimitDegrees=30, twistLimitDegrees=5}
 {
  bone1Name=lowerarm_l
  bone2Name=upperarm_l
  swing1LimitDegrees=45
  swing1Motion=Free
  swing2LimitDegrees=45
  swing2Motion=Free
  twistLimitDegrees=45
  twistMotion=Free
 }
 {
  bone1Name=lowerarm_r
  bone2Name=upperarm_r
  swing1LimitDegrees=45
  swing1Motion=Free
  swing2LimitDegrees=45
  swing2Motion=Free
  twistLimitDegrees=45
  twistMotion=Free
 }
 {
  bone1Name=neck_01
  bone2Name=spine_05
  swing1LimitDegrees=45
  swing1Motion=Free
  swing2LimitDegrees=45
  swing2Motion=Free
  twistLimitDegrees=45
  twistMotion=Free
 }
]
