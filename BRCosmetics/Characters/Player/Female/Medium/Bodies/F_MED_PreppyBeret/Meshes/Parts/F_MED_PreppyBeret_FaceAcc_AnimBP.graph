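# Animation Blueprint for the Preppy Beret cosmetic's face accessory, adding secondary motion to its hair and ponytail.
# Copied body poses feed a backbling-aware Control Rig, then RigidBody or fixed ponytail offsets, then chained AnimDynamics on ponytail and hair bones.
# Despite its name, isRigidBodyEnable selects the fixed offsets, when rigid bodies are off and LOD is 1+.
# The event graph tracks backbling, raised clavicles and skydive gravity/wind settings.
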AnimBlueprint /BRCosmetics/Characters/Player/Female/Medium/Bodies/F_MED_PreppyBeret/Meshes/Parts/F_MED_PreppyBeret_FaceAcc_AnimBP
target_skeleton=&/F_MED_PreppyBeret_FaceAcc_Skeleton
graphs [
 AnimGraph {nodes [
   Copy Pose From Mesh() {bUseAttachedParent=true}
   n1 = Anim Dynamics - Bone: dyn_FrontHair_01!(
    SimulationSpace:AnimPhysSimSpaceType Enum
    ComponentPose=Local To Component(LocalPose=Use cached pose 'PosPonyTailWearingBackpackCache'())
    Alpha=0.2
    ExternalForce="0,0,0"
   ) -> (Pose) {
    angularDampingOverride=1
    angularSpringConstant=12000
    bAngularSpring=true
    bOverrideAngularDamping=true
    bUseGravityOverride=true
    gravityScale=0.35
    simulationSpace=RootRelative
    alphaBoolBlend {blendOption=Linear}
    boundBone {boneName=dyn_FrontHair_01}
    componentAppliedLinearAccClamp=<10000 10000 10000>
    componentLinearAccScale=<1 1 1>
    componentLinearVelScale=<1 1 1>
    gravityOverride {z=-980}
    physicsBodyDefinitions [{
      boundBone {boneName=dyn_FrontHair_01}
      boxExtents=<12 20 12>
      constraintSetup {angularLimitsMax=<5 5 5>, angularLimitsMin=<-2 0 -2>, angularTarget=<1 0 0>}
      localJointOffset=<3 0 0>
     }]
   }
   n2 = Anim Dynamics - Bone: dyn_FrontHair_02!(SimulationSpace:AnimPhysSimSpaceType Enum, ComponentPose=@n1, Alpha=0.4, ExternalForce="0,0,0") -> (Pose) {
    angularDampingOverride=1
    angularSpringConstant=12000
    bAngularSpring=true
    bOverrideAngularDamping=true
    bUseGravityOverride=true
    gravityScale=0.35
    simulationSpace=RootRelative
    alphaBoolBlend {blendOption=Linear}
    boundBone {boneName=dyn_FrontHair_02}
    componentAppliedLinearAccClamp=<10000 10000 10000>
    componentLinearAccScale=<1 1 1>
    componentLinearVelScale=<1 1 1>
    gravityOverride {z=-980}
    physicsBodyDefinitions [{
      boundBone {boneName=dyn_FrontHair_02}
      boxExtents=<12 20 12>
      constraintSetup {angularLimitsMax=<5 5 5>, angularLimitsMin=<-2 0 -2>, angularTarget=<1 0 0>}
      localJointOffset=<3 0 0>
     }]
   }
   n3 = Anim Dynamics - Bone: dyn_sideHair_b_l_01!(SimulationSpace:AnimPhysSimSpaceType Enum, ComponentPose=@n2, Alpha=0.4, ExternalForce="0,0,0") -> (Pose) {
    angularDampingOverride=1
    angularSpringConstant=8995
    bAngularSpring=true
    bOverrideAngularDamping=true
    bUseGravityOverride=true
    gravityScale=0.35
    simulationSpace=RootRelative
    alphaBoolBlend {blendOption=Linear}
    boundBone {boneName=dyn_sideHair_b_l_01}
    componentAppliedLinearAccClamp {y=10000, z=10000}
    componentLinearAccScale=<1 1 1>
    componentLinearVelScale=<1 1 1>
    gravityOverride {z=-980}
    physicsBodyDefinitions [{
      boundBone {boneName=dyn_sideHair_b_l_01}
      boxExtents=<12 20 12>
      constraintSetup {angularLimitsMax=<15 15 15>, angularLimitsMin=<-15 -15 -15>, angularTarget=<1 0 0>}
      localJointOffset=<1 0 0>
     }]
   }
   n4 = Anim Dynamics - Bone: dyn_sideHair_b_l_02!(SimulationSpace:AnimPhysSimSpaceType Enum, ComponentPose=@n3, Alpha=0.4, ExternalForce="0,0,0") -> (Pose) {
    angularDampingOverride=1
    angularSpringConstant=9000
    bAngularSpring=true
    bOverrideAngularDamping=true
    bUseGravityOverride=true
    gravityScale=0.35
    simulationSpace=RootRelative
    alphaBoolBlend {blendOption=Linear}
    boundBone {boneName=dyn_sideHair_b_l_02}
    componentAppliedLinearAccClamp=<10000 10000 10000>
    componentLinearAccScale=<1 1 1>
    componentLinearVelScale=<1 1 1>
    gravityOverride {z=-980}
    physicsBodyDefinitions [{
      boundBone {boneName=dyn_sideHair_b_l_02}
      boxExtents=<12 20 12>
      constraintSetup {angularLimitsMax=<15 15 15>, angularLimitsMin=<-15 -15 -15>, angularTarget=<1 0 0>}
      localJointOffset=<1 0 0>
     }]
   }
   n5 = Anim Dynamics - Bone: dyn_sideHair_b_l_03!(SimulationSpace:AnimPhysSimSpaceType Enum, ComponentPose=@n4, Alpha=0.5, ExternalForce="0,0,0") -> (Pose) {
    angularDampingOverride=1
    angularSpringConstant=6000
    bAngularSpring=true
    bOverrideAngularDamping=true
    bUseGravityOverride=true
    gravityScale=0.35
    simulationSpace=RootRelative
    alphaBoolBlend {blendOption=Linear}
    boundBone {boneName=dyn_sideHair_b_l_03}
    componentAppliedLinearAccClamp=<10000 10000 10000>
    componentLinearAccScale=<0.65 0.65 0.65>
    componentLinearVelScale=<0.65 0.65 0.65>
    gravityOverride {z=-980}
    physicsBodyDefinitions [{
      boundBone {boneName=dyn_sideHair_b_l_03}
      boxExtents=<12 30 12>
      constraintSetup {angularLimitsMax=<5 5 50>, angularLimitsMin=<-5 -5 -30>, angularTarget=<1 0 0>}
      localJointOffset=<4 0 0>
     }]
   }
   n6 = Anim Dynamics - Bone: dyn_sideHair_a_l_01!(SimulationSpace:AnimPhysSimSpaceType Enum, ComponentPose=@n5, Alpha=0.4, ExternalForce="0,0,0") -> (Pose) {
    angularDampingOverride=1
    angularSpringConstant=6000
    bAngularSpring=true
    bOverrideAngularDamping=true
    bUseGravityOverride=true
    gravityScale=0.35
    simulationSpace=RootRelative
    alphaBoolBlend {blendOption=Linear}
    boundBone {boneName=dyn_sideHair_a_l_01}
    componentAppliedLinearAccClamp=<10000 10000 10000>
    componentLinearAccScale=<0.65 0.65 0.65>
    componentLinearVelScale=<0.65 0.65 0.65>
    gravityOverride {z=-980}
    physicsBodyDefinitions [{
      boundBone {boneName=dyn_sideHair_a_l_01}
      boxExtents=<12 30 12>
      constraintSetup {angularLimitsMax=<5 5 50>, angularLimitsMin=<-5 -5 -30>, angularTarget=<1 0 0>}
      localJointOffset=<4 0 0>
     }]
   }
   n7 = Anim Dynamics - Bone: dyn_sideHair_a_l_02!(SimulationSpace:AnimPhysSimSpaceType Enum, ComponentPose=@n6, Alpha=0.5, ExternalForce="0,0,0") -> (Pose) {
    angularDampingOverride=1
    angularSpringConstant=9000
    bAngularSpring=true
    bOverrideAngularDamping=true
    bUseGravityOverride=true
    gravityScale=0.35
    simulationSpace=RootRelative
    alphaBoolBlend {blendOption=Linear}
    boundBone {boneName=dyn_sideHair_a_l_02}
    componentAppliedLinearAccClamp=<10000 10000 10000>
    componentLinearAccScale=<1 1 1>
    componentLinearVelScale=<1 1 1>
    gravityOverride {z=-980}
    physicsBodyDefinitions [{
      boundBone {boneName=dyn_sideHair_a_l_02}
      boxExtents=<12 20 12>
      constraintSetup {angularLimitsMax=<15 15 15>, angularLimitsMin=<-15 -15 -15>, angularTarget=<1 0 0>}
      localJointOffset=<1 0 0>
     }]
   }
   n8 = Anim Dynamics - Bone: dyn_SideHair_l_c_01!(SimulationSpace:AnimPhysSimSpaceType Enum, ComponentPose=@n7, Alpha=1, ExternalForce="0,0,0") -> (Pose) {
    angularDampingOverride=1
    angularSpringConstant=6000
    bAngularSpring=true
    bOverrideAngularDamping=true
    bUseGravityOverride=true
    gravityScale=0.35
    simulationSpace=RootRelative
    alphaBoolBlend {blendOption=Linear}
    boundBone {boneName=dyn_SideHair_l_c_01}
    componentAppliedLinearAccClamp=<10000 10000 10000>
    componentLinearAccScale=<0.65 0.65 0.65>
    componentLinearVelScale=<0.65 0.65 0.65>
    gravityOverride {z=-980}
    physicsBodyDefinitions [{
      boundBone {boneName=dyn_SideHair_l_c_01}
      boxExtents=<12 30 12>
      constraintSetup {angularLimitsMax=<5 5 50>, angularLimitsMin=<-5 -5 -20>, angularTarget=<1 0 0>}
      localJointOffset=<4 0 0>
     }]
   }
   n9 = Anim Dynamics - Bone: dyn_SideHair_l_c_02!(SimulationSpace:AnimPhysSimSpaceType Enum, ComponentPose=@n8, Alpha=1, ExternalForce="0,0,0") -> (Pose) {
    angularDampingOverride=1
    angularSpringConstant=6000
    bAngularSpring=true
    bOverrideAngularDamping=true
    bUseGravityOverride=true
    gravityScale=0.35
    simulationSpace=RootRelative
    alphaBoolBlend {blendOption=Linear}
    boundBone {boneName=dyn_SideHair_l_c_02}
    componentAppliedLinearAccClamp=<10000 10000 10000>
    componentLinearAccScale=<0.65 0.65 0.65>
    componentLinearVelScale=<0.65 0.65 0.65>
    gravityOverride {z=-980}
    physicsBodyDefinitions [{
      boundBone {boneName=dyn_SideHair_l_c_02}
      boxExtents=<12 30 12>
      constraintSetup {angularLimitsMax=<5 5 50>, angularLimitsMin=<-5 -5 -20>, angularTarget=<1 0 0>}
      localJointOffset=<4 0 0>
     }]
   }
   n10 = Anim Dynamics - Bone: dyn_SideHair_l_d_01!(SimulationSpace:AnimPhysSimSpaceType Enum, ComponentPose=@n9, Alpha=1, ExternalForce="0,0,0") -> (Pose) {
    angularDampingOverride=1
    angularSpringConstant=6000
    bAngularSpring=true
    bOverrideAngularDamping=true
    bUseGravityOverride=true
    gravityScale=0.35
    simulationSpace=RootRelative
    alphaBoolBlend {blendOption=Linear}
    boundBone {boneName=dyn_SideHair_l_d_01}
    componentAppliedLinearAccClamp=<10000 10000 10000>
    componentLinearAccScale=<0.65 0.65 0.65>
    componentLinearVelScale=<0.65 0.65 0.65>
    gravityOverride {z=-980}
    physicsBodyDefinitions [{
      boundBone {boneName=dyn_SideHair_l_d_01}
      boxExtents=<12 30 12>
      constraintSetup {angularLimitsMax=<15 15 50>, angularLimitsMin=<-15 -15 -20>, angularTarget=<1 0 0>}
      localJointOffset=<1 0 0>
     }]
   }
   n11 = Anim Dynamics - Bone: dyn_SideHair_l_d_02!(SimulationSpace:AnimPhysSimSpaceType Enum, ComponentPose=@n10, Alpha=1, ExternalForce="0,0,0") -> (Pose) {
    angularDampingOverride=1
    angularSpringConstant=6000
    bAngularSpring=true
    bOverrideAngularDamping=true
    bUseGravityOverride=true
    gravityScale=0.35
    simulationSpace=RootRelative
    alphaBoolBlend {blendOption=Linear}
    boundBone {boneName=dyn_SideHair_l_d_02}
    componentAppliedLinearAccClamp=<10000 10000 10000>
    componentLinearAccScale=<0.65 0.65 0.65>
    componentLinearVelScale=<0.65 0.65 0.65>
    gravityOverride {z=-980}
    physicsBodyDefinitions [{
      boundBone {boneName=dyn_SideHair_l_d_02}
      boxExtents=<12 30 12>
      constraintSetup {angularLimitsMax=<5 5 50>, angularLimitsMin=<-5 -5 -20>, angularTarget=<1 0 0>}
      localJointOffset=<4 0 0>
     }]
   }
   n12 = Anim Dynamics - Bone: dyn_SideHair_l_e_01!(SimulationSpace:AnimPhysSimSpaceType Enum, ComponentPose=@n11, Alpha=0.4, ExternalForce="0,0,0") -> (Pose) {
    angularDampingOverride=1
    angularSpringConstant=9000
    bAngularSpring=true
    bOverrideAngularDamping=true
    bUseGravityOverride=true
    gravityScale=0.35
    simulationSpace=RootRelative
    alphaBoolBlend {blendOption=Linear}
    boundBone {boneName=dyn_SideHair_l_e_01}
    componentAppliedLinearAccClamp=<10000 10000 10000>
    componentLinearAccScale=<1 1 1>
    componentLinearVelScale=<1 1 1>
    gravityOverride {z=-980}
    physicsBodyDefinitions [{
      boundBone {boneName=dyn_SideHair_l_e_01}
      boxExtents=<12 30 12>
      constraintSetup {angularLimitsMax=<5 5 15>, angularLimitsMin=<-5 -5 -15>, angularTarget=<1 0 0>}
      localJointOffset=<4 0 0>
     }]
   }
   n13 = Anim Dynamics - Bone: dyn_SideHair_l_e_02!(SimulationSpace:AnimPhysSimSpaceType Enum, ComponentPose=@n12, Alpha=0.4, ExternalForce="0,0,0") -> (Pose) {
    angularDampingOverride=1
    angularSpringConstant=9000
    bAngularSpring=true
    bOverrideAngularDamping=true
    bUseGravityOverride=true
    gravityScale=0.35
    simulationSpace=RootRelative
    alphaBoolBlend {blendOption=Linear}
    boundBone {boneName=dyn_SideHair_l_e_02}
    componentAppliedLinearAccClamp=<10000 10000 10000>
    componentLinearAccScale=<1 1 1>
    componentLinearVelScale=<1 1 1>
    gravityOverride {z=-980}
    physicsBodyDefinitions [{
      boundBone {boneName=dyn_SideHair_l_e_02}
      boxExtents=<12 20 12>
      constraintSetup {angularLimitsMax=<15 15 15>, angularLimitsMin=<-15 -15 -15>, angularTarget=<1 0 0>}
      localJointOffset=<1 0 0>
     }]
   }
   n14 = Anim Dynamics - Bone: dyn_SideHair_l_e_03!(SimulationSpace:AnimPhysSimSpaceType Enum, ComponentPose=@n13, Alpha=0.5, ExternalForce="0,0,0") -> (Pose) {
    angularDampingOverride=1
    angularSpringConstant=9000
    bAngularSpring=true
    bOverrideAngularDamping=true
    bUseGravityOverride=true
    gravityScale=0.35
    simulationSpace=RootRelative
    alphaBoolBlend {blendOption=Linear}
    boundBone {boneName=dyn_SideHair_l_e_03}
    componentAppliedLinearAccClamp=<10000 10000 10000>
    componentLinearAccScale=<1 1 1>
    componentLinearVelScale=<1 1 1>
    gravityOverride {z=-980}
    physicsBodyDefinitions [{
      boundBone {boneName=dyn_SideHair_l_e_03}
      boxExtents=<12 20 12>
      constraintSetup {angularLimitsMax=<15 15 15>, angularLimitsMin=<-15 -15 -15>, angularTarget=<1 0 0>}
      localJointOffset=<1 0 0>
     }]
   }
   n15 = Anim Dynamics - Bone: dyn_SideHair_r_c_01!(SimulationSpace:AnimPhysSimSpaceType Enum, ComponentPose=@n14, Alpha=0.4, ExternalForce="0,0,0") -> (Pose) {
    angularDampingOverride=1
    angularSpringConstant=9000
    bAngularSpring=true
    bOverrideAngularDamping=true
    bUseGravityOverride=true
    gravityScale=0.35
    simulationSpace=RootRelative
    alphaBoolBlend {blendOption=Linear}
    boundBone {boneName=dyn_SideHair_r_c_01}
    componentAppliedLinearAccClamp=<10000 10000 10000>
    componentLinearAccScale=<1 1 1>
    componentLinearVelScale=<1 1 1>
    gravityOverride {z=-980}
    physicsBodyDefinitions [{
      boundBone {boneName=dyn_SideHair_r_c_01}
      boxExtents=<12 20 12>
      constraintSetup {angularLimitsMax=<15 15 15>, angularLimitsMin=<-15 -15 -15>, angularTarget=<1 0 0>}
      localJointOffset=<1 0 0>
     }]
   }
   n16 = Anim Dynamics - Bone: dyn_SideHair_r_c_02!(SimulationSpace:AnimPhysSimSpaceType Enum, ComponentPose=@n15, Alpha=0.5, ExternalForce="0,0,0") -> (Pose) {
    angularDampingOverride=1
    angularSpringConstant=9000
    bAngularSpring=true
    bOverrideAngularDamping=true
    bUseGravityOverride=true
    gravityScale=0.35
    simulationSpace=RootRelative
    alphaBoolBlend {blendOption=Linear}
    boundBone {boneName=dyn_SideHair_r_c_02}
    componentAppliedLinearAccClamp=<10000 10000 10000>
    componentLinearAccScale=<1 1 1>
    componentLinearVelScale=<1 1 1>
    gravityOverride {z=-980}
    physicsBodyDefinitions [{
      boundBone {boneName=dyn_SideHair_r_c_02}
      boxExtents=<12 20 12>
      constraintSetup {angularLimitsMax=<15 15 15>, angularLimitsMin=<-15 -15 -15>, angularTarget=<1 0 0>}
      localJointOffset=<1 0 0>
     }]
   }
   n17 = Anim Dynamics - Bone: dyn_SideHair_r_b_01!(SimulationSpace:AnimPhysSimSpaceType Enum, ComponentPose=@n16, Alpha=1, ExternalForce="0,0,0") -> (Pose) {
    angularDampingOverride=1
    angularSpringConstant=6000
    bAngularSpring=true
    bOverrideAngularDamping=true
    bUseGravityOverride=true
    gravityScale=0.35
    simulationSpace=RootRelative
    alphaBoolBlend {blendOption=Linear}
    boundBone {boneName=dyn_SideHair_r_b_01}
    componentAppliedLinearAccClamp=<10000 10000 10000>
    componentLinearAccScale=<0.65 0.65 0.65>
    componentLinearVelScale=<0.65 0.65 0.65>
    gravityOverride {z=-980}
    physicsBodyDefinitions [{
      boundBone {boneName=dyn_SideHair_r_b_01}
      boxExtents=<12 30 12>
      constraintSetup {angularLimitsMax=<8 70 10>, angularLimitsMin=<-5 -70 -5>, angularTarget=<1 0 0>}
      localJointOffset=<4 0 0>
     }]
   }
   n18 = Anim Dynamics - Bone: dyn_SideHair_r_b_02!(SimulationSpace:AnimPhysSimSpaceType Enum, ComponentPose=@n17, Alpha=1, ExternalForce="0,0,0") -> (Pose) {
    angularDampingOverride=1
    angularSpringConstant=6000
    bAngularSpring=true
    bOverrideAngularDamping=true
    bUseGravityOverride=true
    gravityScale=0.35
    simulationSpace=RootRelative
    alphaBoolBlend {blendOption=Linear}
    boundBone {boneName=dyn_SideHair_r_b_02}
    componentAppliedLinearAccClamp=<10000 10000 10000>
    componentLinearAccScale=<0.65 0.65 0.65>
    componentLinearVelScale=<0.65 0.65 0.65>
    gravityOverride {z=-980}
    physicsBodyDefinitions [{
      boundBone {boneName=dyn_SideHair_r_b_02}
      boxExtents=<12 30 12>
      constraintSetup {angularLimitsMax=<8 70 10>, angularLimitsMin=<-5 -70 -5>, angularTarget=<1 0 0>}
      localJointOffset=<4 0 0>
     }]
   }
   n19 = Anim Dynamics - Bone: dyn_SideHair_r_d_01!(SimulationSpace:AnimPhysSimSpaceType Enum, ComponentPose=@n18, Alpha=0.4, ExternalForce="0,0,0") -> (Pose) {
    angularDampingOverride=1
    angularSpringConstant=9000
    bAngularSpring=true
    bOverrideAngularDamping=true
    bUseGravityOverride=true
    gravityScale=0.35
    simulationSpace=RootRelative
    alphaBoolBlend {blendOption=Linear}
    boundBone {boneName=dyn_SideHair_r_d_01}
    componentAppliedLinearAccClamp=<10000 10000 10000>
    componentLinearAccScale=<1 1 1>
    componentLinearVelScale=<1 1 1>
    gravityOverride {z=-980}
    physicsBodyDefinitions [{
      boundBone {boneName=dyn_SideHair_r_d_01}
      boxExtents=<12 20 12>
      constraintSetup {angularLimitsMax=<15 15 15>, angularLimitsMin=<-15 -15 -15>, angularTarget=<1 0 0>}
      localJointOffset=<1 0 0>
     }]
   }
   n20 = Anim Dynamics - Bone: dyn_SideHair_r_d_02!(SimulationSpace:AnimPhysSimSpaceType Enum, ComponentPose=@n19, Alpha=0.5, ExternalForce="0,0,0") -> (Pose) {
    angularDampingOverride=1
    angularSpringConstant=9000
    bAngularSpring=true
    bOverrideAngularDamping=true
    bUseGravityOverride=true
    gravityScale=0.35
    simulationSpace=RootRelative
    alphaBoolBlend {blendOption=Linear}
    boundBone {boneName=dyn_SideHair_r_d_02}
    componentAppliedLinearAccClamp=<10000 10000 10000>
    componentLinearAccScale=<1 1 1>
    componentLinearVelScale=<1 1 1>
    gravityOverride {z=-980}
    physicsBodyDefinitions [{
      boundBone {boneName=dyn_SideHair_r_d_02}
      boxExtents=<12 20 12>
      constraintSetup {angularLimitsMax=<15 15 15>, angularLimitsMin=<-15 -15 -15>, angularTarget=<1 0 0>}
      localJointOffset=<1 0 0>
     }]
   }
   n21 = Anim Dynamics - Bone: dyn_SideHair_r_a_01!(SimulationSpace:AnimPhysSimSpaceType Enum, ComponentPose=@n20, Alpha=0.4, ExternalForce="0,0,0") -> (Pose) {
    angularDampingOverride=1
    angularSpringConstant=9000
    bAngularSpring=true
    bOverrideAngularDamping=true
    bUseGravityOverride=true
    gravityScale=0.35
    simulationSpace=RootRelative
    alphaBoolBlend {blendOption=Linear}
    boundBone {boneName=dyn_SideHair_r_a_01}
    componentAppliedLinearAccClamp=<10000 10000 10000>
    componentLinearAccScale=<1 1 1>
    componentLinearVelScale=<1 1 1>
    gravityOverride {z=-980}
    physicsBodyDefinitions [{
      boundBone {boneName=dyn_SideHair_r_a_01}
      boxExtents=<12 20 12>
      constraintSetup {angularLimitsMax=<15 15 15>, angularLimitsMin=<-15 -15 -15>, angularTarget=<1 0 0>}
      localJointOffset=<1 0 0>
     }]
   }
   n22 = Anim Dynamics - Bone: dyn_SideHair_r_a_02!(SimulationSpace:AnimPhysSimSpaceType Enum, ComponentPose=@n21, Alpha=0.5, ExternalForce="0,0,0") -> (Pose) {
    angularDampingOverride=1
    angularSpringConstant=9000
    bAngularSpring=true
    bOverrideAngularDamping=true
    bUseGravityOverride=true
    gravityScale=0.35
    simulationSpace=RootRelative
    alphaBoolBlend {blendOption=Linear}
    boundBone {boneName=dyn_SideHair_r_a_02}
    componentAppliedLinearAccClamp=<10000 10000 10000>
    componentLinearAccScale=<1 1 1>
    componentLinearVelScale=<1 1 1>
    gravityOverride {z=-980}
    physicsBodyDefinitions [{
      boundBone {boneName=dyn_SideHair_r_a_02}
      boxExtents=<12 20 12>
      constraintSetup {angularLimitsMax=<15 15 15>, angularLimitsMin=<-15 -15 -15>, angularTarget=<1 0 0>}
      localJointOffset=<1 0 0>
     }]
   }
   n23 = "Transform (Modify) Bone - Bone: base_SideHair_l_d"!(ComponentPose=@n22, bAlphaBoolEnabled:bool=True, Translation="0,0,0", Rotation="-24.193,0.000,0.000", Scale="1,1,1") -> (Pose) {
    alphaInputType=Bool
    rotationMode=BMM_Additive
    rotationSpace=BCS_BoneSpace
    translationMode=BMM_Additive
    translationSpace=BCS_BoneSpace
    alphaBoolBlend {blendInTime=0.1, blendOption=Linear, blendOutTime=0.1}
    boneToModify {boneName=base_SideHair_l_d}
    rotation=<rot -24.193 0 0>
   }
   Output Pose(Result=Component To Local(ComponentPose=@n23)) {layerGroup=None}
   n24 = Get BodyMesh() -> (BodyMesh:Skeletal Mesh Component Object Reference)
   n25 = Control Rig!(
    Source=Layered blend per bone(BasePose=Copy Pose From Mesh() {bUseAttachedParent=true}, BlendPoses_0=Copy Pose From Mesh(SourceMeshComponent=@n24)) {layerSetup [{branchFilters [{boneName=clavicle_r}, {boneName=clavicle_l}]}]}
    Alpha
    is Backbling:bool=Get IsWearingBackbling() -> (IsWearingBackbling:bool)
    IsRigidBodyDisable:bool=Get isRigidBodyEnable() -> (isRigidBodyEnable:bool)
   ) -> (Pose) {
    alphaBoolBlend {blendOption=Linear}
    controlRigAssetReference {blueprintRigClass=&/F_MED_PreppyBeret_FaceAcc_CtrlRig.F_MED_PreppyBeret_FaceAcc_CtrlRig_C}
   }
   Save cached pose 'InputCache'(Pose=@n25) {cacheName=InputCache}
   n26 = RigidBody!(ComponentPose=Local To Component(LocalPose=Use cached pose 'InputCache'()), Alpha, OverrideWorldGravity:Vector="0,0,0", ExternalForce) -> (Pose) {
    bClampLinearTranslationLimitToRefPose=true
    bOverrideWorldGravity=true
    overridePhysicsAsset=&/F_MED_PreppyBeret_FaceAcc_Physics
    simulationSpace=BaseBoneSpace
    alphaBoolBlend {blendOption=Linear}
    baseBoneRef {boneName=spine_05}
    componentAppliedLinearAccClamp=<8000 8000 8000>
    componentLinearAccScale=<0.3 0.2 0.3>
    componentLinearVelScale=<0.2 0.2 0.3>
    simSpaceSettings {
     maxAngularAcceleration=1000
     maxAngularVelocity=3
     maxLinearAcceleration=700
     maxLinearVelocity=500
     velocityScaleZ=0.45
     worldAlpha=0.6
     externalLinearDragV {x=1}
    }
   }
   n27 = "Transform (Modify) Bone - Bone: dyn_PonyTail_06"(
    ComponentPose=Local To Component(LocalPose=Use cached pose 'InputCache'())
    Alpha=1
    Translation="0,0,2"
    Rotation="-5.001,0.016,0.137"
    Scale="1,1,1"
   ) {
    rotationMode=BMM_Additive
    rotationSpace=BCS_BoneSpace
    translationMode=BMM_Additive
    translationSpace=BCS_BoneSpace
    alphaBoolBlend {blendOption=Linear}
    boneToModify {boneName=dyn_PonyTail_06}
    rotation=<rot -5.001 0.016 0.137>
    translation {z=2}
   }
   n28 = "Transform (Modify) Bone - Bone: dyn_PonyTail_07"(ComponentPose=@n27, Alpha=1, Translation="0,0,0", Rotation="-8,0,0", Scale="1,1,1") {
    rotationMode=BMM_Additive
    rotationSpace=BCS_BoneSpace
    translationMode=BMM_Additive
    translationSpace=BCS_BoneSpace
    alphaBoolBlend {blendOption=Linear}
    boneToModify {boneName=dyn_PonyTail_07}
    rotation {pitch=-8}
   }
   Save cached pose 'posPhysicsDynCache'(Pose=Blend Poses by bool(BlendPose_0=Component To Local(ComponentPose=@n28), BlendPose_1=Component To Local(ComponentPose=@n26), bActiveValue=Get isRigidBodyEnable()) {blendType=Linear}) {cacheName=posPhysicsDynCache}
   n29 = "Transform (Modify) Bone - Bone: dyn_PonyTail_03"!(
    ComponentPose=Local To Component(LocalPose=Use cached pose 'posPhysicsDynCache'())
    bAlphaBoolEnabled:bool=True
    Translation="0,0,-2"
    Rotation="0,0,0"
    Scale="1,1,1"
   ) -> (Pose) {
    alphaInputType=Bool
    rotationMode=BMM_Additive
    rotationSpace=BCS_BoneSpace
    translationMode=BMM_Replace
    translationSpace=BCS_BoneSpace
    alphaBoolBlend {blendInTime=0.25, blendOption=Linear, blendOutTime=0.25}
    boneToModify {boneName=dyn_PonyTail_03}
    translation {z=-2}
   }
   n30 = Anim Dynamics - Bone: dyn_PonyTail_07!(ComponentPose=@n29, Alpha=1, GravityOverride:Vector="0,0,-980", ExternalForce="0,0,0") -> (Pose) {
    angularDampingOverride=1
    angularSpringConstant=3000
    bAngularSpring=true
    bOverrideAngularDamping=true
    bUseGravityOverride=true
    bUseSphericalLimits=true
    gravityScale=0.35
    simulationSpace=RootRelative
    alphaBoolBlend {blendOption=Linear}
    boundBone {boneName=dyn_PonyTail_07}
    componentAppliedLinearAccClamp=<10000 10000 10000>
    componentLinearAccScale=<0.25 0.25 0.25>
    componentLinearVelScale=<0.25 0.25 0.25>
    gravityOverride {z=-980}
    physicsBodyDefinitions [{
      boundBone {boneName=dyn_PonyTail_07}
      boxExtents=<12 30 12>
      constraintSetup {angularLimitsMax=<0 10 10>, angularLimitsMin=<-2 -10 -10>, angularTarget=<1 0 0>}
      localJointOffset=<4 0 0>
     }]
    sphericalLimits [{limitRadius=19, drivingBone {boneName=spine_01}, sphereLocalOffset=<-2 8 0>}]
   }
   Save cached pose 'PosPonyTailWearingBackpackCache'(Pose=Component To Local(ComponentPose=@n30)) {cacheName=PosPonyTailWearingBackpackCache}
  ]}
 EventGraph {nodes [
   Event BlueprintUpdateAnimation()
   > n1 = Sequence!(execute) -> (then_0, then_1, then_2:Exec)
   > n2 = Branch(Condition=Get IsWearingBackbling())
   > n3 = Set members in Gravity Override Params Struct!(
    execute:Exec
    StructRef:"Gravity Override Params Struct Structure (by ref)"=Get Gravity Override Params()
    Skydive_GoingUp_28_#1:Gravity Override Setting Struct Structure=Make Gravity Override Setting Struct!(
     WindFrequency_28_#2:Vector="1.5,5,6"
     WindAmplitude_31_#3:Vector="0,70,0"
     WindOffset_32_#4:Vector="0,0,0"
     GravityOverride_15_#5:Vector="0,-300,-100"
     JointName_27_#6:Name=root
    ) -> (GravityOverrideSettingStruct:Gravity Override Setting Struct Structure) {is=title K2Node_MakeStruct}
    Skydive_FloatingDown_29_#7:Gravity Override Setting Struct Structure=Make Gravity Override Setting Struct(
     WindFrequency_28_#2="1.5,7,8"
     WindAmplitude_31_#3="200,-500,400"
     WindOffset_32_#4="0,-600,0"
     GravityOverride_15_#5="0,-300,-600"
     JointName_27_#6=root
    )
    Skydive_DivingDown_30_#8:Gravity Override Setting Struct Structure=Make Gravity Override Setting Struct(
     WindFrequency_28_#2="1.5,12,10"
     WindAmplitude_31_#3="0,50,50"
     WindOffset_32_#4="0,0,0"
     GravityOverride_15_#5="-1500,0,0"
     JointName_27_#6=pelvis
    )
    Skydive_Parachuting_31_#9:Gravity Override Setting Struct Structure=Make Gravity Override Setting Struct(
     WindFrequency_28_#2="1.5,7,8"
     WindAmplitude_31_#3="50,-100,400"
     WindOffset_32_#4="0,0,0"
     GravityOverride_15_#5="0,-300,-980"
     JointName_27_#6=root
    )
   ) -> (then:Exec, StructOut:"Gravity Override Params Struct Structure (by ref)") {is=title K2Node_SetFieldsInStruct}
   > Set Gravity Override Params(Gravity Override Params:Gravity Override Params Struct Structure=@n3) -> (Output_Get:Gravity Override Params Struct Structure)
   n4 = Set members in Gravity Override Params Struct(
    execute=@n2.else
    StructRef=Get Gravity Override Params()
    Skydive_GoingUp_28_#1=Make Gravity Override Setting Struct(
     WindFrequency_28_#2="1.5,7,8"
     WindAmplitude_31_#3="200,-500,400"
     WindOffset_32_#4="0,0,0"
     GravityOverride_15_#5="0,-300,1500"
     JointName_27_#6=root
    )
    Skydive_FloatingDown_29_#7=Make Gravity Override Setting Struct(
     WindFrequency_28_#2="1.5,7,8"
     WindAmplitude_31_#3="200,-500,400"
     WindOffset_32_#4="0,-600,400"
     GravityOverride_15_#5="0,-700,600"
     JointName_27_#6=root
    )
    Skydive_DivingDown_30_#8=Make Gravity Override Setting Struct(
     WindFrequency_28_#2="1.5,12,10"
     WindAmplitude_31_#3="50,300,100"
     WindOffset_32_#4="0,0,0"
     GravityOverride_15_#5="-1500,0,0"
     JointName_27_#6=pelvis
    )
    Skydive_Parachuting_31_#9=Make Gravity Override Setting Struct(
     WindFrequency_28_#2="1.5,7,8"
     WindAmplitude_31_#3="200,-500,400"
     WindOffset_32_#4="0,0,0"
     GravityOverride_15_#5="0,-800,-980"
     JointName_27_#6=root
    )
   )
   > Set Gravity Override Params(Gravity Override Params=@n4)
   n5 = Gravity Override State Function DEPRECATED(
    execute=@n1.then_1
    CharacterPartAnimInstance=Self-Reference()
    GameTimeInSeconds=Get Game Time in Seconds() -> (Game Time in Seconds:double)
    GravityOverrideParams=Get Gravity Override Params()
   )
   > Set GravityOverride(GravityOverride=@n5)
   n6 = Get bRigidBodyEnabled()
   n7 = Get CurrentLOD()
   > Set isRigidBodyEnable(isRigidBodyEnable:bool=AND Boolean(A=NOT Boolean(A=@n6), B="integer >= integer"(A=@n7, B=1))) -> (Output_Get:bool)
   n8 = Get BodyMesh()
   n9 = BreakTransform(InTransform=GetDeltaTransformFromRefPose(self=@n8, BoneName=clavicle_r, BaseName=spine_05))
   Set r_clavicle_up(execute=@n1.then_2, r_clavicle_up:bool="float >= float"(A=BreakRotator(InRot=@n9.Rotation).Pitch, B=35)) -> (Output_Get:bool)
   n10 = BreakTransform(InTransform=GetDeltaTransformFromRefPose(self=@n8, BoneName=clavicle_l, BaseName=spine_05))
   > Set l_clavicle_up(l_clavicle_up:bool="float >= float"(A=BreakRotator(InRot=@n10.Rotation).Pitch, B=35)) -> (Output_Get:bool)
   n11 = Event OnCharacterCustomizationLoaded()
   > n12 = Sequence!(execute) -> (then_0, then_1)
   > Set BodyMesh(BodyMesh=Get Array Item with EFort Custom Part Type(Array=@n11.CustomMeshParts))
   > n13 = Branch(Condition=Get bIsInFrontEnd())
   > Set Sim Space Hair(Sim Space Hair:AnimPhysSimSpaceType Enum=World) -> (Output_Get:AnimPhysSimSpaceType Enum)
   > Set HairDynAlpha(HairDynAlpha:double=0.4) -> (Output_Get:double)
   Set Sim Space Hair(execute=@n13.else, Sim Space Hair=BoneRelative)
   > Set HairDynAlpha(HairDynAlpha=1)
   n14 = CheckForBackbling()
   n15 = Set IsWearingBackbling(execute=@n12.then_1, IsWearingBackbling:bool=@n14) -> (Output_Get:bool)
   > Set Is Backbling A Cape(Is Backbling A Cape:bool=@n14.bIsBackblingACape) -> (Output_Get:bool)
   > Set NotBackbling(NotBackbling:double="To Float (Boolean)"(InBool=NOT Boolean(A=@n14))) -> (Output_Get:double)
   > Set is Wear a Cape(is Wear a Cape:bool=AND Boolean(A=@n14, B=@n14.bIsBackblingLong)) -> (Output_Get:bool)
   > n16 = Branch(Condition=@n15)
   > n17 = Branch(Condition=@n14.bIsBackblingLong)
   Set IsShortBackblingOnAlpha(execute=[@n17.then, @n16.else], IsShortBackblingOnAlpha:double=0) -> (Output_Get:double)
   Set IsShortBackblingOnAlpha(execute=@n17.else, IsShortBackblingOnAlpha=1)
  ]}
]
variables [
 BodyMesh {type=SkeletalMeshComponent}
 Game Time in Seconds {type=number}
 Gravity Override Params {type=GravityOverrideParamsStruct}
 GravityOverride {type=Vector}
 HairDynAlpha {type=number}
 Is Backbling A Cape {type=boolean}
 is Wear a Cape {type=boolean}
 isRigidBodyEnable {type=boolean}
 IsShortBackblingOnAlpha {type=number}
 IsWearingBackbling {type=boolean}
 l_clavicle_up {type=boolean}
 NotBackbling {type=number}
 r_clavicle_up {type=boolean}
 Sim Space Hair {type=AnimPhysSimSpaceType}
]
